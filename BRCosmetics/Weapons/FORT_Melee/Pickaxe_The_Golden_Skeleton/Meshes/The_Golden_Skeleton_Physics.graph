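# Physics asset for The Golden Skeleton pickaxe, a two-handed melee cosmetic, simulating two dangling chains attached to its handle.
# The handle body uses two capsules, and each chain link has its own box body.
# Chain A has three links and chain B has five, each constrained to the previous link or the handle.
# All joints share moderate swing limits and a tight twist limit, so the chains sway with little twisting.
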PhysicsAsset /BRCosmetics/Weapons/FORT_Melee/Pickaxe_The_Golden_Skeleton/Meshes/The_Golden_Skeleton_Physics
bodies {
 Melee_TwoHanded_Handle [
  {shapeType=Capsule, length=130.625, radius=3, center=<0 0 30.4989>}
  {shapeType=Capsule, length=4.5355, radius=8.2553, center=<0 -0.55 54.8293>}
 ]
 dyn_chain_01_a [{shapeType=Box, extentX=4.3862, extentY=1.499, extentZ=6.5428, center=<3 0 0>, rotation=<rot 90 0 0>}]
 dyn_chain_01_b [{shapeType=Box, extentX=4.3862, extentY=1.499, extentZ=6.5428, center=<3 0 0>, rotation=<rot 90 0 0>}]
 dyn_chain_02_a [{shapeType=Box, extentX=1.7987, extentY=4.3695, extentZ=6.5428, center=<3 0 0>, rotation=<rot 90 0 0>}]
 dyn_chain_02_b [{shapeType=Box, extentX=1.7987, extentY=4.3695, extentZ=6.5428, center=<3 0 0>, rotation=<rot 90 0 0>}]
 dyn_chain_03_a [{shapeType=Box, extentX=4.4395, extentY=2.3758, extentZ=6.6799, center=<3 0 0>, rotation=<rot 90 0 0>}]
 dyn_chain_03_b [{shapeType=Box, extentX=4.3197, extentY=2.0835, extentZ=6.5428, center=<3 0 0>, rotation=<rot 90 0 0>}]
 dyn_chain_04_b [{shapeType=Box, extentX=1.7987, extentY=4.3695, extentZ=6.5428, center=<3 0 0>, rotation=<rot 90 0 0>}]
 dyn_chain_05_b [{shapeType=Box, extentX=4.3093, extentY=4.1701, extentZ=6.6799, center=<3 0 0>, rotation=<rot 90 0 0>}]
}
constraints [
 {bone1Name=dyn_chain_03_a, bone2Name=dyn_chain_02_a, swing1LimitDegrees=45, swing2LimitDegrees=45, twistLimitDegrees=6}
 {bone1Name=dyn_chain_01_a, bone2Name=Melee_TwoHanded_Handle, swing1LimitDegrees=45, swing2LimitDegrees=45, twistLimitDegrees=6}
 {bone1Name=dyn_chain_02_a, bone2Name=dyn_chain_01_a, swing1LimitDegrees=45, swing2LimitDegrees=45, twistLimitDegrees=6}
 {bone1Name=dyn_chain_01_b, bone2Name=Melee_TwoHanded_Handle, swing1LimitDegrees=45, swing2LimitDegrees=45, twistLimitDegrees=6}
 {bone1Name=dyn_chain_02_b, bone2Name=dyn_chain_01_b, swing1LimitDegrees=45, swing2LimitDegrees=45, twistLimitDegrees=6}
 {bone1Name=dyn_chain_03_b, bone2Name=dyn_chain_02_b, swing1LimitDegrees=45, swing2LimitDegrees=45, twistLimitDegrees=6}
 {bone1Name=dyn_chain_04_b, bone2Name=dyn_chain_03_b, swing1LimitDegrees=45, swing2LimitDegrees=45, twistLimitDegrees=6}
 {bone1Name=dyn_chain_05_b, bone2Name=dyn_chain_04_b, swing1LimitDegrees=45, swing2LimitDegrees=45, twistLimitDegrees=6}
]
